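# Default Lit material for the rubber duck quest interactable, with duck diffuse, normal and mask textures plus a procedural bobbing motion.
# A sine of time and BobSpeed drives both a small rocking rotation about X and a vertical offset, summed into what appears to be a vertex offset.
# Object position X and Y sample a time ramp texture to scale the sine's frequency, so ducks in different places bob at different rates.
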Material /Figment_S10/Athena/Items/QuestInteractables/RubberDuckies/FX/M_RubberDuck_QuestInteractable
graph {
 properties {bUsedWithNanite=true, pixelDepthOffsetMode=PDOM_Legacy, usedShadingModels=Default Lit}
 nodes [
  n1 = Constant() {r=-0.01}
  n2 = Constant() {r=0.01}
  n3 = Time()
  n4 = ScalarParameter() {defaultValue=10, parameterName=BobSpeed}
  n5 = Divide(A=ObjectPositionWS(), B=Constant() {r=55})
  n6 = TextureSample(UVs=ComponentMask(None=@n5) {r=true}) {samplerType=SAMPLERTYPE_Masks, texture=/Game/Environments/Sets/Arid/Textures/T_TimeRamp}
  n7 = TextureSample(UVs=ComponentMask(None=@n5) {g=true}) {samplerType=SAMPLERTYPE_Masks, texture=&T_TimeRamp}
  n8 = Sine(None=Multiply(A=Multiply(A=@n3, B=@n4), B=Multiply(A=@n6.R, B=@n7.R)))
  n9 = ObjectPositionWS()
  n10 = WorldPosition()
  Add(
   A=RotateAboutAxis(
    NormalizedRotationAxis=Constant3Vector() {constant=<rgba 1 0 0 0>}
    RotationAngle=LinearInterpolate(A=@n1, B=@n2, Alpha=@n8)
    PivotPoint=Constant3Vector()
    Position=Subtract(A=@n9, B=@n10.XYZ)
   )
   B=LinearInterpolate(A=Constant3Vector() {constant=<rgba 0 0 1 1>}, B=Constant3Vector() {constant=<rgba 0 0 -2 1>}, Alpha=@n8)
  )
  TextureSample() {texture=/Game/Athena/Items/QuestInteractables/RubberDuckies/FX/Quest_Rubber_Duck_D}
  TextureSample() {samplerType=SAMPLERTYPE_Normal, texture=/Game/Athena/Items/QuestInteractables/RubberDuckies/FX/Quest_Rubber_Duck_N}
  TextureSample() {samplerType=SAMPLERTYPE_Masks, texture=/Game/Athena/Items/QuestInteractables/RubberDuckies/FX/Quest_Rubber_Duck_S}
 ]
}
