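; Big rubber ducky quest prop that registers player visits to its location and occasionally quacks when damaged.
; Five seconds after BeginPlay it binds a sphere overlap that sends a location-tagged stat event to the overlapping player's quest manager.
; Server damage has a one-in-twenty chance to trigger a client-only quack sound at the actor, then a timer holds off further quacks.
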(event EventBeginPlay
  (Utilities|FlowControl|Delay 5.0)
  (Collision|BindEventtoOnComponentBeginOverlap (Variables|Default|GetSphere) (delegate Custom|VisitedDucky)))

(event Custom|VisitedDucky
  (params (OverlappedComponent :type Primitive%20Component%20Object%20Reference) (OtherActor :type Actor%20Object%20Reference) (OtherComp :type Primitive%20Component%20Object%20Reference) (OtherBodyIndex :type Integer) (bFromSweep :type Boolean) (SweepResult :type Hit%20Result%20Structure%20%28by%20ref%29))
  (bind _returnvalue (Fortnite|GetFortPlayerControllerfromActor OtherActor))
  ;[1 1 1] Quest Visit hook in
  (Utilities|IsValid _returnvalue
    (:Is%20Valid
      (Quest|SendComplexCustomStatEvent (Quests|GetQuestManager _returnvalue "Invalid") 0 "(GameplayTags=((TagName=\"Athena.Quests.Location.FDS.BigRubberDucky\")))" "(GameplayTags=)" 1))))

(event EventOnDamageServer
  (params (Damage :type Float%20%28single-precision%29) (DamageTags :type Gameplay%20Tag%20Container%20Structure%20%28by%20ref%29) (Momentum :type Vector :default "0, 0, 0") (HitInfo :type Hit%20Result%20Structure%20%28by%20ref%29) (InstigatedBy :type Controller%20Object%20Reference) (DamageCauser :type Actor%20Object%20Reference) (EffectContext :type Gameplay%20Effect%20Context%20Handle%20Structure))
  ;[1 1 1] Check to see if QUACK should play
  (if (Variables|Default|GetReadyToQuack)
    (if (== (Math|Random|RandomInteger 20) 1)
      (CallFunction|PlayQuackonClients)
      (Utilities|Time|SetTimerbyEvent (delegate Custom|StopFurtherQuacking) 1e+01)
      (Variables|Default|SetReadyToQuack false))))

(event Custom|StopFurtherQuacking
  ;[1 1 1] Check to see if QUACK should play
  (Variables|Default|SetReadyToQuack true))

(event Custom|PlayQuackOnClients
  ;[1 1 1] QUACK sound - client only
  (Networking|SwitchHasAuthority
    (:Remote
      (Audio|PlaySoundatLocation (Variables|Default|GetQuackSound) (Transformation|GetActorLocation)))))

(fn ConstructionScript)
